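; PG_ActorCounter is a Playgrounds utility that audits Fortnite building actors in the level through two custom events.
; Collect Total Actor Counts tallies actors and unique classes for all BuildingSMActors, structural pieces and props, then prints totals.
; It then subtracts structural and prop classes from the overall list and prints leftovers in orange as unwanted.
; Collect All Materials gathers unique wall and floor texture data; stair and roof loops lack array inputs.
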(event Custom|Collect%20Total%20Actor%20Counts
  (CallFunction|InitializeVariables)
  (CallFunction|DetermineandSetActorandClassCount)
  (CallFunction|SetUniqueClassCount)
  (CallFunction|SetTotalCountText)
  (CallFunction|DetermineUnwantedActors))

(event Custom|Collect%20All%20Materials
  (bind _uniquematerial_array (Variables|Default|GetUniqueMaterial_Array))
  (Utilities|Array|Clear _uniquematerial_array)
  (CallFunction|GatherUniqueMaterals "/Script/FortniteGame.BuildingWall" _uniquematerial_array)
  (CallFunction|GatherUniqueMaterals "/Script/FortniteGame.BuildingStairs" _uniquematerial_array)
  (CallFunction|GatherUniqueMaterals "/Script/FortniteGame.BuildingRoof" _uniquematerial_array)
  (CallFunction|GatherUniqueMaterals "/Script/FortniteGame.BuildingFloor" _uniquematerial_array))

(fn CountTotalActors
  (params (ActorClass :type Actor%20Class%20Reference) (OutputActorCount :type Integer%20%28by%20ref%29) (UniqueActorClassesArray :type Array%20of%20Actor%20Class%20References))
  (bind _outactors (Actor|GetAllActorsOfClass ActorClass))
  (for _index (range (+ (Utilities|Array|LastIndex _outactors) 1))
    (Math|Integer|IncrementInt (Variables|GetOutputActorCount)
      (:%20%20
        (Utilities|Array|AddUnique (Variables|GetUniqueActorClassesArray) (Utilities|GetClass (Utilities|Array|Get(acopy) _outactors _index)))))))

(fn DetermineAndSetActorAndClassCount
  ;[0.4383 1 0.1161] Total Actors & Classes
  (CallFunction|CountTotalActors "/Script/FortniteGame.BuildingSMActor" (Variables|Default|GetTotalBuildingSMActors) (Variables|Default|GetUniqueBuildingSMActorClasses_Array))
  ;[0.09495 1 0.8196] Building Structural Actors & Classes (Building Wall, Building Stairs, Building Floor, Building Roof, Building Corner)
  (CallFunction|CountTotalActors "/Script/FortniteGame.BuildingWall" (Variables|Default|GetTotalBuildingStructuralActors) (Variables|Default|GetUniqueBuildingStructuralActorClasses_Array))
  (CallFunction|CountTotalActors "/Script/FortniteGame.BuildingStairs" (Variables|Default|GetTotalBuildingStructuralActors) (Variables|Default|GetUniqueBuildingStructuralActorClasses_Array))
  (CallFunction|CountTotalActors "/Script/FortniteGame.BuildingFloor" (Variables|Default|GetTotalBuildingStructuralActors) (Variables|Default|GetUniqueBuildingStructuralActorClasses_Array))
  (CallFunction|CountTotalActors "/Script/FortniteGame.BuildingRoof" (Variables|Default|GetTotalBuildingStructuralActors) (Variables|Default|GetUniqueBuildingStructuralActorClasses_Array))
  (CallFunction|CountTotalActors "/Script/FortniteGame.BuildingCorner" (Variables|Default|GetTotalBuildingStructuralActors) (Variables|Default|GetUniqueBuildingStructuralActorClasses_Array))
  ;[0 0.152 1] Prop Actors & Classes  (Building Prop, Building Container, Building Deco)
  (CallFunction|CountTotalActors "/Script/FortniteGame.BuildingProp" (Variables|Default|GetTotalPropActors) (Variables|Default|GetUniqueBuildingPropClasses_Array))
  (CallFunction|CountTotalActors "/Script/FortniteGame.BuildingContainer" (Variables|Default|GetTotalPropActors) (Variables|Default|GetUniqueBuildingPropClasses_Array))
  (CallFunction|CountTotalActors "/Script/FortniteGame.BuildingDeco" (Variables|Default|GetTotalPropActors) (Variables|Default|GetUniqueBuildingPropClasses_Array)))

(fn DetermineUnwantedActors
  (for _array_element (Variables|Default|GetUniqueBuildingPropClasses_Array)
    (Utilities|Array|RemoveItem (Variables|Default|GetUniqueBuildingSMActorClasses_Array) _array_element))
  (for _array_element_2 (Variables|Default|GetUniqueBuildingStructuralActorClasses_Array)
    (Utilities|Array|RemoveItem (Variables|Default|GetUniqueBuildingSMActorClasses_Array) _array_element_2))
  (for _array_element_3 (Variables|Default|GetUniqueBuildingSMActorClasses_Array)
    (Development|PrintString (Utilities|GetClassDisplayName _array_element_3) true true "(R=1.000000,G=0.352687,B=0.000000,A=1.000000)")))

(fn GatherUniqueMaterals
  (params (ActorClass :type Actor%20Class%20Reference) (Texture%20Data :type Array%20of%20Building%20Texture%20Data%20Object%20References))
  (bind _outactors (Actor|GetAllActorsOfClass ActorClass))
  (for _array_element _outactors
    (bind _asbuilding_wall (Utilities|Casting|CastToBuildingWall _array_element)
      (:then
        (bind _outtexturedataarray (Building|GetBuildingTextureData _asbuilding_wall :!class /Script/FortniteGame.BuildingSMActor))
        (for _array_element_2 _outtexturedataarray
          (Utilities|Array|AddUnique (Variables|GetTexture%20Data) _array_element_2)))
      (:CastFailed
        (bind _asbuilding_floor (Utilities|Casting|CastToBuildingFloor _array_element)
          (:then
            (bind _outtexturedataarray_2 (Building|GetBuildingTextureData _asbuilding_floor :!class /Script/FortniteGame.BuildingSMActor))
            (for _array_element_3 _outtexturedataarray_2
              (Utilities|Array|AddUnique (Variables|GetTexture%20Data) _array_element_3)))
          (:CastFailed
            (bind _asbuilding_stairs (Utilities|Casting|CastToBuildingStairs _array_element)
              (:then
                (Building|GetBuildingTextureData _asbuilding_stairs :!class /Script/FortniteGame.BuildingSMActor)
                (for _array_element_4 0
                  (Utilities|Array|AddUnique (Variables|GetTexture%20Data) _array_element_4)))
              (:CastFailed
                (bind _asbuilding_roof (Utilities|Casting|CastToBuildingRoof _array_element))
                (Building|GetBuildingTextureData _asbuilding_roof :!class /Script/FortniteGame.BuildingSMActor)
                (for _array_element_5 0
                  (Utilities|Array|AddUnique (Variables|GetTexture%20Data) _array_element_5))))))))))

(fn InitializeVariables
  (Utilities|FlowControl|Sequence
    (:then_0
      (Variables|Default|SetTotalBuildingStructuralActors 0))
    (:then_1
      (Variables|Default|SetTotalUniqueBuildingStructuralClasses 0))
    (:then_2
      (Variables|Default|SetTotalPropActors 0))
    (:then_3
      (Variables|Default|SetTotalUniquePropClasses 0))
    (:then_4
      (Variables|Default|SetTotalBuildingSMActors 0))
    (:then_5
      (Utilities|Array|Clear (Variables|Default|GetUniqueBuildingStructuralActorClasses_Array)))
    (:then_6
      (Utilities|Array|Clear (Variables|Default|GetUniqueBuildingPropClasses_Array)))
    (:then_7
      (Utilities|Array|Clear (Variables|Default|GetUniqueBuildingSMActorClasses_Array)))))

(fn SetFormatText
  (params (Count%20Type :type Text) (InInt :type Integer))
  (Development|PrintString (Utilities|String|ToString(Text) (Utilities|Text|FormatText "NSLOCTEXT(\"[EDF13A1AEAA06789C2064CA4E3BC49FA]\", \"758FE0BF46F290E248D9808F6BB72C7F\", \"{Count Type}  {Amount}\")" Count%20Type (Utilities|Text|ToText(String) (Utilities|String|ToString(Integer) InInt))))))

(fn SetTotalCountText
  ;[0.4383 1 0.1161] Total Actors & Classes
  (CallFunction|SetFormatText (Variables|Default|GetTotalBuildingSMActors_Text) (Variables|Default|GetTotalBuildingSMActors))
  (CallFunction|SetFormatText (Variables|Default|GetTotalUniqueBuildingSMActorClasses_Text) (Variables|Default|GetTotalUniqueBuildingSMActorClasses))
  ;[0.09495 1 0.8196] Building Structural Actors & Classes
  (CallFunction|SetFormatText (Variables|Default|GetTotalBuildingStructurealActors_Text) (Variables|Default|GetTotalBuildingStructuralActors))
  (CallFunction|SetFormatText (Variables|Default|GetTotalUniqueBuildingStrcturealActors_Text) (Variables|Default|GetTotalUniqueBuildingStructuralClasses))
  ;[0 0.152 1] Prop Actors & Classes
  (CallFunction|SetFormatText (Variables|Default|GetTotalPropActors_Text) (Variables|Default|GetTotalPropActors))
  (CallFunction|SetFormatText (Variables|Default|GetTotalUniquePropClasses_Text) (Variables|Default|GetTotalUniquePropClasses)))

(fn SetUniqueClassCount
  ;[0.4383 1 0.1161] Total Actors & Classes
  (Variables|Default|SetTotalUniquePropClasses (Utilities|Array|Length (Variables|Default|GetUniqueBuildingPropClasses_Array)))
  ;[0.09495 1 0.8196] Building Structural Actors & Classes
  (Variables|Default|SetTotalUniqueBuildingStructuralClasses (Utilities|Array|Length (Variables|Default|GetUniqueBuildingStructuralActorClasses_Array)))
  ;[0 0.152 1] Prop Actors & Classes
  (Variables|Default|SetTotalUniqueBuildingSMActorClasses (Utilities|Array|Length (Variables|Default|GetUniqueBuildingSMActorClasses_Array))))

(fn ConstructionScript)
